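# Translucent unlit deferred decal material for a fire scorch mark with animated glowing detail.
# Smoke noise panning in radial coordinates distorts the UVs of two noise textures, whose product is masked by an inverted soft sphere mask to weaken the center.
# That pattern is tinted by an orange HDR Color parameter and passed through MF_FX_EmissiveAdjustments.
# A separate scorch-hole texture is multiplied by decal lifetime opacity and a scalar boost for fading.
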Material /Game/Effects/Niagara/MasterMaterials/Fire/RealisticFire/MD_Fire_ScorchDecal
graph {
 properties {
  bAutomaticallySetUsageInEditor=false
  blendMode=BLEND_Translucent
  materialDomain=MD_DeferredDecal
  shadingModel=MSM_Unlit
  usedShadingModels=Unlit
 }
 nodes [
  Constant()
  Constant()
  n1 = Panner(Coordinate=Multiply(
    A=MaterialFunctionCall() {materialFunction=&VectorToRadialValue}.Radial Coordinates
    B=Constant2Vector() {g=1, r=3}
   )) {speedY=-0.05}
  n2 = Add(
   A=Multiply(A=TextureCoordinate()) {constB=2}
   B=Divide(A=TextureSample(UVs=@n1) {texture=/Game/Global/Textures/Default/Tileable_Patterns/T_Thick_Smoke_05}.R) {constB=7}
  )
  n3 = TextureSample(UVs=@n2) {texture=/Game/Weapons/WeaponSkins/Textures/T_Swirly_Noise_highContrast}
  n4 = TextureCoordinate()
  n5 = Constant() {r=0.5}
  n6 = Multiply(
   A=Multiply(A=@n3.R, B=TextureSample(UVs=@n2) {samplerType=SAMPLERTYPE_Grayscale, texture=/Game/Global/Textures/Default/Noise/TilingNoise06}.R)
   B=OneMinus(None=SphereMask(A=@n4, B=@n5) {attenuationRadius=0.5, hardnessPercent=0})
  )
  MaterialFunctionCall(Emissive=Multiply(A=Multiply(A=@n6), B=VectorParameter() {parameterName=Color, defaultValue=<rgba 5 0.5977 0.1026 1>}.RGB)) {materialFunction=&MF_FX_EmissiveAdjustments}
  Multiply(
   A=ScalarParameter() {defaultValue=1.5}
   B=Multiply(
    A=TextureSample() {texture=/Game/Effects/Fort_Effects/Gameplay/SmokeDustExplosions/Textures/T_Decal_Scorch_Hole}.R
    B=DecalLifetimeOpacity().Opacity
   )
  )
  Multiply(A=@n3.R, B=@n3.R)
 ]
}
