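# Translucent unlit Niagara master material for fire trails on ribbons and mesh particles, with optional spline thickening.
# Two panning flame noises, distorted by swirl noise on high-end platforms and timed by clock or dynamic-parameter curve, blend two fire colors.
# UV gradients and a sphere mask shape the noise into opacity, with particle-alpha erosion, depth fade and optional view-angle fade.
# Dynamic parameters also drive emissive, width and random UV offset.
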Material /Game/Effects/Niagara/MasterMaterials/Fire/RealisticFire/M_FireTrail_Ribbon
graph {
 properties {
  bAutomaticallySetUsageInEditor=false
  bUsedWithNiagaraMeshParticles=true
  bUsedWithNiagaraRibbons=true
  blendMode=BLEND_Translucent
  shadingModel=MSM_Unlit
  usedShadingModels=Unlit
 }
 nodes [
  n1 = Add(A=DynamicParameter() {paramNames [PanSpeedCurve, Width, Emissive, RandomUV]}.RandomUV, B=ScalarParameter() {parameterName=RandomUVoffset})
  n2 = MaterialFunctionCall() {materialFunction=&LinearGradient}
  n3 = ScalarParameter() {defaultValue=2, parameterName=BiasU_Tile}
  n4 = Multiply(A=AppendVector(A=Power(Base=@n2.UGradient, Exp=@n3), B=@n2.VGradient), B=Constant2Vector() {g=0.6, r=0.6})
  n5 = VectorParameter() {parameterName=TileDivider, defaultValue=<rgba 0.75 0.75 0.5 0.5>}
  n6 = VectorParameter() {parameterName=PanDirection, defaultValue=<rgba -1 0.2 -1 -0.2>}
  n7 = ScalarParameter() {defaultValue=1, parameterName=PanSpeed}
  n8 = DynamicParameter() {paramNames [PanSpeedCurve, Width, Emissive, RandomUV]}
  n9 = LinearInterpolate(B=ParticleRandom(), Alpha=ScalarParameter() {defaultValue=1, parameterName=UseParticleRandomTimeForPanning})
  n10 = LinearInterpolate(
   A=Multiply(A=Time(), B=@n7)
   B=Multiply(A=Add(A=@n8.PanSpeedCurve, B=@n9), B=@n7)
   Alpha=ScalarParameter() {defaultValue=1, parameterName=UseCurvedPanning}
  )
  n11 = Panner(Coordinate=Divide(A=Add(A=@n1, B=@n4), B=AppendVector(A=@n5.R, B=@n5.R)) {constB=0.75}, Time=@n10, Speed=AppendVector(A=@n6.R, B=@n6.R)) {speedX=-1, speedY=0.2}
  n12 = Constant()
  n13 = TextureCoordinate()
  n14 = Constant2Vector() {g=0.6, r=0.6}
  n15 = TextureSample(UVs=Panner(Coordinate=Multiply(A=@n13, B=@n14), Time=@n10) {speedX=0.5, speedY=0.1}) {samplerType=SAMPLERTYPE_Normal, texture=/Game/Effects/Fort_Effects/Emotes/BirdNestBlue/T_SwirlNoise}
  n16 = MaterialFunctionCall!(
   "Xb1,PS4"=@n12
   "PC-High&&Epic,XSX,PS5"=Divide(A=ComponentMask(None=@n15.RGB) {g=true, r=true}, B=ScalarParameter() {defaultValue=15, parameterName=RefractionIntensityDivider}) {constB=15}
   "PC-Perf,Android"=@n12
   Switch=@n12
   "PC-Low(NotPerf)&&Medium"=@n12
  ) {materialFunction=/Game/Effects/Niagara/MasterMaterials/Functions/Common/PlatformSwitch_2D}
  n17 = Panner(Coordinate=Divide(A=Add(A=@n1, B=@n4), B=AppendVector(A=@n5.B, B=@n5.B)) {constB=0.5}, Time=@n10, Speed=AppendVector(A=@n6.B, B=@n6.B)) {speedX=-1, speedY=-0.2}
  n18 = Add(
   A=TextureSample(UVs=Add(A=@n11, B=@n16.Result) {constB=0}) {samplerType=SAMPLERTYPE_LinearGrayscale, texture=/Game/Effects/Fort_Effects/Emotes/UltraLight/Texture/T_FluidFlameTile_Greyscale}.R
   B=TextureSample(UVs=Add(A=@n17, B=@n16.Result)) {samplerType=SAMPLERTYPE_LinearGrayscale, texture=/Game/Effects/Fort_Effects/Emotes/UltraLight/Texture/T_DirectionalNoise_Greyscale}.R
  )
  n19 = Saturate(None=@n18)
  n20 = LinearInterpolate(
   A=Multiply(
    A=VectorParameter() {parameterName=FireColorB, defaultValue=<rgba 1 0.0529 0.0002 1>}.RGB
    B=ScalarParameter() {defaultValue=20, parameterName=FireColorB_Mutilplier}
   )
   B=Multiply(
    A=VectorParameter() {parameterName=FireColorA, defaultValue=<rgba 1 0.3075 0.0527 1>}.RGB
    B=ScalarParameter() {defaultValue=10, parameterName=FireColorA_Mutilplier}
   )
   Alpha=@n19
  )
  n21 = ParticleColor()
  MaterialFunctionCall(Emissive=Multiply(A=DynamicParameter() {paramNames [PanSpeedCurve, Width, Emissive, RandomUV]}.Emissive, B=Multiply(A=@n20, B=@n21.RGB))) {materialFunction=&MF_FX_EmissiveAdjustments}
  n22 = TextureCoordinate()
  n23 = OneMinus(None=@n22)
  n24 = ScalarParameter() {parameterName=RevertUTile}
  n25 = MaterialFunctionCall(XY=@n22) {materialFunction=&Swizzle}
  n26 = MaterialFunctionCall(UV Channel=LinearInterpolate(A=LinearInterpolate(A=@n22, B=@n23, Alpha=@n24), B=OneMinus(None=@n25.YX), Alpha=ScalarParameter() {parameterName=UseSwizzle})) {materialFunction=&LinearGradient}
  n27 = Multiply(A=@n26.UGradient, B=@n26.UGradient)
  n28 = Saturate(None=Multiply(A=OneMinus(None=@n26.UGradient)) {constB=10})
  n29 = LinearInterpolate(A=@n27, B=Multiply(A=Multiply(A=@n28, B=@n27)) {constB=2}, Alpha=ScalarParameter() {defaultValue=1, parameterName=FadeOutTheLeadingEdge}) {constAlpha=1}
  n30 = Saturate(None=Sine(None=Multiply(A=@n26.VGradient) {constB=0.5}))
  n31 = Constant2Vector() {g=0.5}
  n32 = Constant2Vector() {g=1, r=0.5}
  n33 = ScalarParameter() {parameterName=UseSwizzle}
  n34 = Multiply(
   A=Saturate(None=Multiply(A=@n29, B=@n30))
   B=SphereMask(A=TextureCoordinate(), B=LinearInterpolate(A=@n31, B=@n32, Alpha=@n33)) {attenuationRadius=1.1, hardnessPercent=0.8}
  )
  n35 = Add(A=Multiply(A=@n18, B=@n34), B=@n34)
  n36 = TextureCoordinate() {coordinateIndex=1}
  n37 = Multiply(A=OneMinus(None=MaterialFunctionCall(UV Channel=@n36) {materialFunction=&LinearGradient}.UGradient))
  n38 = ScalarParameter() {parameterName=UseUV1toFadeOutTheLeadingEdge}
  n39 = Multiply(A=Multiply(A=@n19, B=@n21.A), B=Multiply(A=Multiply(A=@n35, B=@n35), B=LinearInterpolate(B=@n37, Alpha=@n38) {constA=1}))
  n40 = Subtract(A=@n39, B=LinearInterpolate(A=Constant() {r=0.2}, Alpha=@n21.A) {constA=0.05, constB=0})
  n41 = ScalarParameter() {defaultValue=1, parameterName=UseParticleAlphaToDoErosion}
  n42 = DepthFade(Opacity=Saturate(None=LinearInterpolate(A=@n39, B=@n40, Alpha=@n41))) {fadeDistanceDefault=150}
  n43 = Constant() {r=0.05}
  n44 = LinearInterpolate(
   B=OneMinus(None=MaterialFunctionCall("FadeSensitivity (optional)"=@n43) {materialFunction=&FadeBasedOnViewAngle}.Result)
   Alpha=ScalarParameter() {parameterName=UseCamdeAngleFade}
  )
  Multiply(A=Saturate(None=Subtract(A=@n42, B=@n44)), B=ScalarParameter() {defaultValue=1, parameterName=Opacity})
  n45 = Multiply(
   A=ScalarParameter() {defaultValue=50, parameterName=Width}
   B=DynamicParameter() {paramNames [PanSpeedCurve, Width, Emissive, RandomUV]}.Width
  )
  StaticSwitchParameter(
   True=MaterialFunctionCall(WidthBase=@n45, WidthTip=@n45) {materialFunction=&SplineThicken}.WorldPositionOffset
   False=Constant()
  ) {parameterName=UseSplineMesh}
 ]
}
