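# Translucent unlit master material for Niagara sprite particles that renders flipbook fire from a packed RGB texture.
# Particle relative time or a dynamic-parameter curve advances the FlipBook function from an optionally randomized start phase.
# A channel-picker parameter selects the mask, which with particle color blends low-to-high fire colors, with desaturation, into emissive adjustments.
# Opacity combines mask and particle alpha with mask-varied depth fade and near-camera fade.
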Material /Game/Effects/Niagara/MasterMaterials/Fire/RealisticFire/M_Flipbook_Fire
graph {
 properties {
  bAutomaticallySetUsageInEditor=false
  bUsedWithNiagaraSprites=true
  blendMode=BLEND_Translucent
  shadingModel=MSM_Unlit
  usedShadingModels=Unlit
 }
 nodes [
  Constant3Vector() {constant=<rgba 1 0 0 0>}
  Constant3Vector() {constant=<rgba 0 1 0 0>}
  Constant3Vector() {constant=<rgba 0 0 1 0>}
  n1 = ParticleColor()
  n2 = ScalarParameter() {defaultValue=0.2, parameterName=AnimPhaseStartOffset}
  n3 = Constant() {r=2}
  n4 = LinearInterpolate(
   A=@n2
   B=LinearInterpolate(A=Divide(A=@n2, B=@n3) {constB=4}, B=Multiply(A=@n2, B=@n3) {constB=4}, Alpha=ParticleRandom())
   Alpha=ScalarParameter() {parameterName=RandomAnimPhaseStartOffset}
  )
  n5 = LinearInterpolate(
   A=ParticleRelativeTime()
   B=DynamicParameter() {paramNames [FlipAnimCurve, Param2, Param3, Param4]}.FlipAnimCurve
   Alpha=ScalarParameter() {desc=If LW Emitter use this, parameterName=UseFlipAnimCurve}
  )
  n6 = ScalarParameter() {defaultValue=8, parameterName=NumberOfRowAndColumns}
  n7 = TextureSampleParameter2D(UVs=MaterialFunctionCall("Animation  Phase (0-1)"=LinearInterpolate(A=@n4, Alpha=@n5), Number of Rows=@n6, Number of Columns=@n6) {materialFunction=&FlipBook}.UVs) {parameterName=Texture, texture=/Game/Effects/Fort_Effects/Textures/Fire/T_Flipbook_Fire_RGB_Packed}
  n8 = Multiply(A=ScalarParameter() {desc=(R=0, G=0.5, B=1), parameterName=Texture_RGB_Picker}) {constA=0.5, constB=2}
  n9 = LinearInterpolate(A=LinearInterpolate(A=@n7.R, B=@n7.R, Alpha=Saturate(None=@n8)), B=@n7.B, Alpha=Saturate(None=Subtract(A=@n8)))
  n10 = Multiply(A=@n1.RGB, B=@n9)
  n11 = LinearInterpolate(
   A=VectorParameter() {parameterName=FireColorLow, defaultValue=<rgba 1 0.0994 0.0004 1>}.RGB
   B=Multiply(
    A=VectorParameter() {parameterName=FireColorHigh, defaultValue=<rgba 1 0.2879 0.0856 1>}.RGB
    B=ScalarParameter() {defaultValue=3, parameterName=FireColorHigh_Multiplier}
   ) {constB=3}
   Alpha=@n10
  )
  n12 = ScalarParameter() {parameterName=Desaturation}
  MaterialFunctionCall(Emissive=Multiply(A=Desaturation(None=@n11, Fraction=@n12), B=@n10)) {materialFunction=&MF_FX_EmissiveAdjustments}
  Round(None=ParticleRandom())
  n13 = Multiply(A=@n1.A, B=@n9)
  n14 = ScalarParameter() {defaultValue=20, parameterName=DepthFade}
  n15 = LinearInterpolate(A=Multiply(A=@n14) {constB=10}, B=@n14, Alpha=@n9) {constA=200, constB=20}
  n16 = MaterialFunctionCall(Distance=ScalarParameter() {defaultValue=180, parameterName=NearCameraFadeDistance}) {materialFunction=&Near_Camera_Fade}
  Saturate(None=Subtract(A=DepthFade(Opacity=@n13, FadeDistance=@n15), B=OneMinus(None=@n16.Result)))
  Time()
 ]
}
